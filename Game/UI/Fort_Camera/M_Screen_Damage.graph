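# M_Screen_Damage is an additive unlit particle material for a screen damage effect in the camera UI.
# It multiplies the T_Screen_Damage texture color by particle alpha and a dynamic parameter, so particle systems can drive its intensity.
# The result is clamped with a high ceiling and fed as emissive into the MF_FX_EmissiveAdjustments function.
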Material /Game/UI/Fort_Camera/M_Screen_Damage
graph {
 properties {bUsedWithParticleSprites=true, blendMode=BLEND_Additive, pixelDepthOffsetMode=PDOM_Legacy, shadingModel=MSM_Unlit, usedShadingModels=Unlit}
 nodes [
  n1 = Multiply(A=TextureSample() {texture=&/T_Screen_Damage}.RGB, B=ParticleColor().A)
  n2 = DynamicParameter()
  MaterialFunctionCall(Emissive=Clamp(None=Multiply(A=@n1, B=@n2.Param1)) {maxDefault=500}) {materialFunction=&MF_FX_EmissiveAdjustments}
 ]
}
