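# Additive, unlit particle material that, judging by its name, draws a screen-space shield effect, reusing the screen damage texture.
# Screen-aligned UVs fitted to the texture's 2:1 dimensions are scaled about the center to sample the texture's red channel as a mask.
# Particle color and alpha tint and fade the mask, and a dynamic parameter multiplies its intensity.
# The clamped result passes through the shared FX emissive adjustments function.
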Material /Game/UI/Fort_Camera/M_Screen_Sheild
graph {
 properties {bUsedWithParticleSprites=true, blendMode=BLEND_Additive, pixelDepthOffsetMode=PDOM_Legacy, shadingModel=MSM_Unlit, usedShadingModels=Unlit}
 nodes [
  n1 = MaterialFunctionCall(
   UVs=MaterialFunctionCall(Texture Dimensions=Constant2Vector() {g=512, r=1024}) {materialFunction=&ScreenAlignedUVs}."X 100%, Y 100%"
   Texture Scale=Constant2Vector() {g=1.2, r=1.5}
  ) {materialFunction=&ScaleUVsByCenter}
  n2 = ParticleColor()
  n3 = Multiply(A=Multiply(A=TextureSample(UVs=@n1.UVs) {texture=&/T_Screen_Damage}.R, B=@n2.RGB), B=@n2.A)
  n4 = DynamicParameter()
  MaterialFunctionCall(Emissive=Clamp(None=Multiply(A=@n3, B=@n4.Param1)) {maxDefault=500}) {materialFunction=&MF_FX_EmissiveAdjustments}
 ]
}
